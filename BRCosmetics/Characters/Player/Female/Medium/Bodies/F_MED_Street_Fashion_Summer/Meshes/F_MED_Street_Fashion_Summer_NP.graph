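# Physics asset for the necklace on the female medium Street Fashion Summer character body, giving its dyn_necklace bones capsule bodies for physical motion.
# A large capsule on spine_05 represents the upper torso and anchors both the top necklace bone and a three-bone bottom chain.
# Constraints allow no twist and only small swing at the anchors, with swing loosening down the bottom chain toward its tip.
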PhysicsAsset /BRCosmetics/Characters/Player/Female/Medium/Bodies/F_MED_Street_Fashion_Summer/Meshes/F_MED_Street_Fashion_Summer_NP
bodies {
 dyn_necklace_bot_01 [{shapeType=Capsule, length=8.2092, radius=0.75, center=<2.3224 0.5728 0.0245>, rotation=<rot -0.5292 -89.9997 1.4744>}]
 dyn_necklace_bot_02 [{shapeType=Capsule, length=5.073, radius=1, center=<1.7386 0.3387 0.0697>, rotation=<rot -0.6332 -90 0.2033>}]
 dyn_necklace_bot_03 [{shapeType=Capsule, length=3.0543, radius=0.8715, center=<2.0847 0.317 0>, rotation=<rot 90 17.354 14.5416>}]
 dyn_necklace_top [{shapeType=Capsule, length=4.7433, radius=0.5, center=<2.1732 -0.2502 -0.3377>, rotation=<rot -1.46 -90 3.4891>}]
 spine_05 [{shapeType=Capsule, length=14.0084, radius=10.9057, center=<-3.1124 -0.868 0>, rotation=<rot 0 -35.0002 0>}]
}
constraints [
 {bone1Name=dyn_necklace_top, bone2Name=spine_05, swing1LimitDegrees=2, swing2LimitDegrees=2, twistLimitDegrees=0}
 {bone1Name=dyn_necklace_bot_01, bone2Name=spine_05, swing1LimitDegrees=5, swing2LimitDegrees=5, twistLimitDegrees=0}
 {bone1Name=dyn_necklace_bot_02, bone2Name=dyn_necklace_bot_01, swing1LimitDegrees=10, swing2LimitDegrees=5, twistLimitDegrees=0}
 {bone1Name=dyn_necklace_bot_03, bone2Name=dyn_necklace_bot_02, swing1LimitDegrees=25, swing2LimitDegrees=25, twistLimitDegrees=0}
]
