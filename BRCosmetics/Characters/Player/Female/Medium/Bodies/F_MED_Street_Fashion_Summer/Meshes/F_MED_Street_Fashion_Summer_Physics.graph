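# Physics asset for the female medium Street Fashion Summer body, covering the dress's left-side dynamic bones and tie.
# Front, back and side chains of three capsules hang from small MSTR spheres, with tight twist and limited swing.
# Those spheres link to the left thigh-twist capsule by angular-free constraints, as do two side segments.
# Two tie capsules share their own master sphere, and a right-thigh capsule has no constraints.
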PhysicsAsset /BRCosmetics/Characters/Player/Female/Medium/Bodies/F_MED_Street_Fashion_Summer/Meshes/F_MED_Street_Fashion_Summer_Physics
bodies {
 dyn_dress_back_01_l [{shapeType=Capsule, length=15.5051, radius=2.441, center=<2.6618 0 1.5011>}]
 dyn_dress_back_01_l_MSTR [{shapeType=Sphere, radius=0.505}]
 dyn_dress_back_02_l [{shapeType=Capsule, length=14.9648, radius=2.5179, center=<2.8445 0 0>, rotation=<rot 0 0 10.0003>}]
 dyn_dress_back_03_l [{shapeType=Capsule, length=14.295, radius=2.5908, center=<4.0497 0 0>, rotation=<rot 0 0 20.0019>}]
 dyn_dress_front_01_l [{shapeType=Capsule, length=9.021, radius=1.5236, center=<2.8911 0.1819 0>}]
 dyn_dress_front_01_l_MSTR [{shapeType=Sphere, radius=0.505}]
 dyn_dress_front_02_l [{shapeType=Capsule, length=9.2452, radius=1.6578, center=<2.2798 1 0>}]
 dyn_dress_front_03_l [{shapeType=Capsule, length=7.3164, radius=1.1285, center=<1.5576 0.4219 -0.0001>}]
 dyn_dress_side_01_l [{shapeType=Capsule, length=7.4363, radius=1.6331, center=<2.829 0 -0.0004>, rotation=<rot 0 0 90>}]
 dyn_dress_side_01_l_MSTR [{shapeType=Sphere, radius=0.505}]
 dyn_dress_side_02_l [{shapeType=Capsule, length=6.1032, radius=1.1401, center=<1.631 0 0>, rotation=<rot 0 0 90>}]
 dyn_dress_side_03_l [{shapeType=Capsule, length=5.6635, radius=0.8488, center=<1.5702 0 0>, rotation=<rot 0 0 90>}]
 dyn_dress_tie_MSTR [{shapeType=Sphere, radius=0.505}]
 dyn_dress_tie_b [{shapeType=Capsule, length=6.73, radius=1.8179, center=<9.8173 -3.5392 -2.9381>, rotation=<rot 0 -81.88 -77.9599>}]
 dyn_dress_tie_f [{shapeType=Capsule, length=8.5583, radius=2.2857, center=<10.0559 3.6337 -1.3869>, rotation=<rot 0 -99.4755 -77.2684>}]
 thigh_r [{shapeType=Capsule, length=27.0856, radius=8.0785, center=<24.1952 0 0>, rotation=<rot 90 0 0>}]
 thigh_twist_01_l [{shapeType=Capsule, length=27.4606, radius=7.7035, center=<4.6512 -0.3914 0.4629>, rotation=<rot 90 -0.0192 -0.0187>}]
}
constraints [
 {
  bone1Name=dyn_dress_front_01_l_MSTR
  bone2Name=thigh_twist_01_l
  swing1LimitDegrees=45
  swing1Motion=Free
  swing2LimitDegrees=45
  swing2Motion=Free
  twistLimitDegrees=45
  twistMotion=Free
 }
 {bone1Name=dyn_dress_front_01_l, bone2Name=dyn_dress_front_01_l_MSTR, swing1LimitDegrees=25, swing2LimitDegrees=25, twistLimitDegrees=5}
 {bone1Name=dyn_dress_front_02_l, bone2Name=dyn_dress_front_01_l, swing1LimitDegrees=25, swing2LimitDegrees=25, twistLimitDegrees=5}
 {bone1Name=dyn_dress_front_03_l, bone2Name=dyn_dress_front_02_l, swing1LimitDegrees=25, swing2LimitDegrees=25, twistLimitDegrees=5}
 {
  bone1Name=dyn_dress_back_01_l_MSTR
  bone2Name=thigh_twist_01_l
  swing1LimitDegrees=45
  swing1Motion=Free
  swing2LimitDegrees=45
  swing2Motion=Free
  twistLimitDegrees=45
  twistMotion=Free
 }
 {bone1Name=dyn_dress_back_01_l, bone2Name=dyn_dress_back_01_l_MSTR, swing1LimitDegrees=35, swing2LimitDegrees=35, twistLimitDegrees=5}
 {bone1Name=dyn_dress_back_02_l, bone2Name=dyn_dress_back_01_l, swing1LimitDegrees=35, swing2LimitDegrees=35, twistLimitDegrees=5}
 {bone1Name=dyn_dress_back_03_l, bone2Name=dyn_dress_back_02_l, swing1LimitDegrees=35, swing2LimitDegrees=35, twistLimitDegrees=5}
 {
  bone1Name=dyn_dress_side_01_l_MSTR
  bone2Name=thigh_twist_01_l
  swing1LimitDegrees=45
  swing1Motion=Free
  swing2LimitDegrees=45
  swing2Motion=Free
  twistLimitDegrees=45
  twistMotion=Free
 }
 {bone1Name=dyn_dress_side_01_l, bone2Name=dyn_dress_side_01_l_MSTR, swing1LimitDegrees=65, swing2LimitDegrees=45, twistLimitDegrees=5}
 {bone1Name=dyn_dress_side_02_l, bone2Name=dyn_dress_side_01_l, swing1LimitDegrees=25, swing2LimitDegrees=25, twistLimitDegrees=5}
 {bone1Name=dyn_dress_side_03_l, bone2Name=dyn_dress_side_02_l, swing1LimitDegrees=25, swing2LimitDegrees=25, twistLimitDegrees=5}
 {bone1Name=dyn_dress_tie_b, bone2Name=dyn_dress_tie_MSTR, swing1LimitDegrees=10, swing2LimitDegrees=55, twistLimitDegrees=5}
 {bone1Name=dyn_dress_tie_f, bone2Name=dyn_dress_tie_MSTR, swing1LimitDegrees=10, swing2LimitDegrees=55, twistLimitDegrees=5}
 {
  bone1Name=thigh_twist_01_l
  bone2Name=dyn_dress_side_01_l
  swing1LimitDegrees=45
  swing1Motion=Free
  swing2LimitDegrees=45
  swing2Motion=Free
  twistLimitDegrees=45
  twistMotion=Free
 }
 {
  bone1Name=thigh_twist_01_l
  bone2Name=dyn_dress_side_02_l
  swing1LimitDegrees=45
  swing1Motion=Free
  swing2LimitDegrees=45
  swing2Motion=Free
  twistLimitDegrees=45
  twistMotion=Free
 }
]
