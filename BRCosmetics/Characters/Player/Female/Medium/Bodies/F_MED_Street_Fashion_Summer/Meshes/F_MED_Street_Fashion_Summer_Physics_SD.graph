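# Physics asset for the female medium Street Fashion Summer body, providing secondary motion for the dress's left-side panels and tie.
# Front, back and side panels are three-box chains with limited swing and tight twist, each rooted on a small master sphere.
# Those masters attach to the left thigh twist bone with free rotation, while two tie capsules share their own master sphere.
# Capsules on both thighs stand in for the legs.
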PhysicsAsset /BRCosmetics/Characters/Player/Female/Medium/Bodies/F_MED_Street_Fashion_Summer/Meshes/F_MED_Street_Fashion_Summer_Physics_SD
bodies {
 dyn_dress_back_01_l [{shapeType=Box, extentX=5.75, extentY=1.25, extentZ=8.25, center=<8.9373 0 0.0001>}]
 dyn_dress_back_01_l_MSTR [{shapeType=Sphere, radius=0.505}]
 dyn_dress_back_02_l [{shapeType=Box, extentX=5.75, extentY=3.75, extentZ=8.25, center=<6.2492 0.0001 0>}]
 dyn_dress_back_03_l [{shapeType=Box, extentX=5.75, extentY=4, extentZ=8.25, center=<4.6159 0 0>}]
 dyn_dress_front_01_l [{shapeType=Box, extentX=5.75, extentY=1.75, extentZ=10.5, center=<4.6378 0 0>}]
 dyn_dress_front_01_l_MSTR [{shapeType=Sphere, radius=0.505}]
 dyn_dress_front_02_l [{shapeType=Box, extentX=5.75, extentY=1.75, extentZ=10.5, center=<3.3316 0 0>}]
 dyn_dress_front_03_l [{shapeType=Box, extentX=5.75, extentY=1.75, extentZ=10.5, center=<3.3316 0 0>}]
 dyn_dress_side_01_l [{shapeType=Box, extentX=4.75, extentY=9.25, extentZ=1.5, center=<5.4106 0 0.0003>}]
 dyn_dress_side_01_l_MSTR [{shapeType=Sphere, radius=0.505}]
 dyn_dress_side_02_l [{shapeType=Box, extentX=4.75, extentY=9.25, extentZ=1.5, center=<5.4106 0 0.0003>}]
 dyn_dress_side_03_l [{shapeType=Box, extentX=4.75, extentY=9.25, extentZ=1.5, center=<3.3308 0 0.0001>}]
 dyn_dress_tie_MSTR [{shapeType=Sphere, radius=0.505}]
 dyn_dress_tie_b [{shapeType=Capsule, length=6.73, radius=1.8179, center=<9.8173 -3.5392 -2.9381>, rotation=<rot 0 -81.88 -77.9599>}]
 dyn_dress_tie_f [{shapeType=Capsule, length=8.3083, radius=2.5357, center=<10.0559 3.6337 -1.3869>, rotation=<rot 0 -99.4755 -77.2684>}]
 thigh_r [{shapeType=Capsule, length=27.0856, radius=8.0785, center=<21.7811 0 0>, rotation=<rot 90 0 0>}]
 thigh_twist_01_l [{shapeType=Capsule, length=27.0856, radius=8.0785, center=<4.6512 -0.3914 -0.0002>, rotation=<rot 90 -0.0192 -0.0187>}]
}
constraints [
 {
  bone1Name=dyn_dress_front_01_l_MSTR
  bone2Name=thigh_twist_01_l
  swing1LimitDegrees=45
  swing1Motion=Free
  swing2LimitDegrees=45
  swing2Motion=Free
  twistLimitDegrees=45
  twistMotion=Free
 }
 {bone1Name=dyn_dress_front_01_l, bone2Name=dyn_dress_front_01_l_MSTR, swing1LimitDegrees=45, swing2LimitDegrees=45, twistLimitDegrees=5}
 {bone1Name=dyn_dress_front_02_l, bone2Name=dyn_dress_front_01_l, swing1LimitDegrees=15, swing2LimitDegrees=15, twistLimitDegrees=5}
 {bone1Name=dyn_dress_front_03_l, bone2Name=dyn_dress_front_02_l, swing1LimitDegrees=20, swing2LimitDegrees=20, twistLimitDegrees=5}
 {
  bone1Name=dyn_dress_back_01_l_MSTR
  bone2Name=thigh_twist_01_l
  swing1LimitDegrees=45
  swing1Motion=Free
  swing2LimitDegrees=45
  swing2Motion=Free
  twistLimitDegrees=45
  twistMotion=Free
 }
 {bone1Name=dyn_dress_back_01_l, bone2Name=dyn_dress_back_01_l_MSTR, swing1LimitDegrees=25, swing2LimitDegrees=25, twistLimitDegrees=5}
 {bone1Name=dyn_dress_back_02_l, bone2Name=dyn_dress_back_01_l, swing1LimitDegrees=35, swing2LimitDegrees=25, twistLimitDegrees=5}
 {bone1Name=dyn_dress_back_03_l, bone2Name=dyn_dress_back_02_l, swing1LimitDegrees=40, swing2LimitDegrees=30, twistLimitDegrees=5}
 {
  bone1Name=dyn_dress_side_01_l_MSTR
  bone2Name=thigh_twist_01_l
  swing1LimitDegrees=45
  swing1Motion=Free
  swing2LimitDegrees=45
  swing2Motion=Free
  twistLimitDegrees=45
  twistMotion=Free
 }
 {bone1Name=dyn_dress_side_01_l, bone2Name=dyn_dress_side_01_l_MSTR, swing1LimitDegrees=10, swing2LimitDegrees=10, twistLimitDegrees=5}
 {bone1Name=dyn_dress_side_02_l, bone2Name=dyn_dress_side_01_l, swing1LimitDegrees=10, swing2LimitDegrees=25, twistLimitDegrees=5}
 {bone1Name=dyn_dress_side_03_l, bone2Name=dyn_dress_side_02_l, swing1LimitDegrees=15, swing2LimitDegrees=30, twistLimitDegrees=5}
 {bone1Name=dyn_dress_tie_b, bone2Name=dyn_dress_tie_MSTR, swing1LimitDegrees=5, swing2LimitDegrees=45, twistLimitDegrees=5}
 {bone1Name=dyn_dress_tie_f, bone2Name=dyn_dress_tie_MSTR, swing1LimitDegrees=5, swing2LimitDegrees=45, twistLimitDegrees=5}
]
