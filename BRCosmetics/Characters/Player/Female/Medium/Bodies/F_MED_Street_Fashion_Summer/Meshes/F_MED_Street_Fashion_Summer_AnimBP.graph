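# Anim Blueprint for the female Street Fashion Summer body, adding secondary motion to the dress, tie and grenade.
# It copies the attached parent's pose, then layers RigidBody physics, bone offsets and constraints, and grenade AnimDynamics into a cached pose.
# A final RigidBody pass feeds the output, weakened while skydiving; the separate skydiving-physics blend is unconnected.
# The EventGraph derives dress-tie corrective alphas from right-thigh rotation, branching on emote playback.
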AnimBlueprint /BRCosmetics/Characters/Player/Female/Medium/Bodies/F_MED_Street_Fashion_Summer/Meshes/F_MED_Street_Fashion_Summer_AnimBP
target_skeleton=&/F_MED_Street_Fashion_Summer_Skeleton
graphs [
 AnimGraph {nodes [
   n1 = RigidBody(ComponentPose=Local To Component(LocalPose=Use cached pose 'SavedPose3377'())) {
    bOverrideWorldGravity=true
    overridePhysicsAsset=&/F_MED_Street_Fashion_Summer_Physics_SD
    simulationSpace=BaseBoneSpace
    alphaBoolBlend {blendOption=Linear}
    baseBoneRef {boneName=pelvis}
    componentAppliedLinearAccClamp=<3000 3000 3000>
    componentLinearAccScale=<0.3 0.3 0.3>
    componentLinearVelScale=<0.2 0.2 0.2>
    overrideWorldGravity {z=-400}
   }
   Blend Poses by bool(BlendPose_0=Component To Local(ComponentPose=@n1), bActiveValue=Get bIsSkyDiving()) {blendType=Linear}
   "Transform (Modify) Bone - Bone: dyn_dress_front_01_l_MSTR"!(ComponentPose, bAlphaBoolEnabled:bool=True, Translation="2,0,0", Rotation="20,0,0", Scale="1,1,1") -> (Pose) {
    alphaInputType=Bool
    rotationMode=BMM_Additive
    rotationSpace=BCS_BoneSpace
    translationMode=BMM_Additive
    translationSpace=BCS_BoneSpace
    alphaBoolBlend {blendOption=Linear}
    boneToModify {boneName=dyn_dress_front_01_l_MSTR}
    rotation {pitch=20}
    translation {x=2}
   }
   "Transform (Modify) Bone - Bone: dyn_dress_side_01_l_MSTR"(Alpha=1, Translation="0,2,0", Rotation="0.000,-20,0.000", Scale="1,1,1") {
    rotationMode=BMM_Additive
    rotationSpace=BCS_BoneSpace
    translationMode=BMM_Additive
    translationSpace=BCS_BoneSpace
    alphaBoolBlend {blendOption=Linear}
    boneToModify {boneName=dyn_dress_side_01_l_MSTR}
    rotation {yaw=-20}
    translation {y=2}
   }
   n2 = RigidBody(ComponentPose=Local To Component(LocalPose=Use cached pose 'SavedPose3377'()), Alpha=Get SD Alpha() -> (SD Alpha:double)) {
    bOverrideWorldGravity=true
    overridePhysicsAsset=&/F_MED_Street_Fashion_Summer_Physics
    simulationSpace=BaseBoneSpace
    alphaBoolBlend {blendOption=Linear}
    baseBoneRef {boneName=root}
    componentAppliedLinearAccClamp=<5000 5000 5000>
    componentLinearAccScale=<0.33 0.33 0.5>
    componentLinearVelScale=<0.2 0.2 0.2>
    overrideWorldGravity {z=-980}
   }
   Output Pose(Result=Component To Local(ComponentPose=@n2)) {layerGroup=None}
   n3 = RigidBody(ComponentPose=Local To Component(LocalPose=Copy Pose From Mesh() {bUseAttachedParent=true})) {
    bOverrideWorldGravity=true
    overridePhysicsAsset=&/F_MED_Street_Fashion_Summer_NP
    simulationSpace=BaseBoneSpace
    alphaBoolBlend {blendOption=Linear}
    baseBoneRef {boneName=root}
    componentLinearAccScale=<0.05 0.05 0.05>
    componentLinearVelScale=<0.05 0.05 0.05>
    overrideWorldGravity {z=-500}
   }
   n4 = "Transform (Modify) Bone - Bone: dyn_dress_tie_MSTR"(ComponentPose=@n3, Alpha=Get Tie Side Alpha() -> (Tie Side Alpha:double), Translation="0,0,0", Rotation="-15,0,0", Scale="1,1,1") {
    rotationMode=BMM_Additive
    rotationSpace=BCS_BoneSpace
    alphaBoolBlend {blendOption=Linear}
    boneToModify {boneName=dyn_dress_tie_MSTR}
    rotation {pitch=-15}
   }
   n5 = "Transform (Modify) Bone - Bone: dyn_dress_tie_MSTR"(ComponentPose=@n4, Alpha=Get Tie Front Alpha() -> (Tie Front Alpha:double), Translation="0,0,0", Rotation="0,20,-35", Scale="1,1,1") {
    rotationMode=BMM_Additive
    rotationSpace=BCS_BoneSpace
    alphaBoolBlend {blendOption=Linear}
    boneToModify {boneName=dyn_dress_tie_MSTR}
    rotation {roll=-35, yaw=20}
   }
   n6 = "Transform (Modify) Bone - Bone: dyn_dress_tie_MSTR"(ComponentPose=@n5, Alpha=Get Tie Front EM Alpha() -> (Tie Front EM Alpha:double), Translation="0,0,0", Rotation="0,20,-60", Scale="1,1,1") {
    rotationMode=BMM_Additive
    rotationSpace=BCS_BoneSpace
    alphaBoolBlend {blendOption=Linear}
    boneToModify {boneName=dyn_dress_tie_MSTR}
    rotation {roll=-60, yaw=20}
   }
   n7 = "Constraint - dyn_grenade_MSTR (PPP)"!(ComponentPose=@n6, Alpha=1, ConstraintWeights_0=0.6365, ConstraintWeights_1:float=0.035, ConstraintWeights_2:float=0.3286) -> (Pose) {
    alphaBoolBlend {blendOption=Linear}
    boneToModify {boneName=dyn_grenade_MSTR}
    constraintSetup [{targetBone {boneName=thigh_r}}, {targetBone {boneName=pelvis}}, {targetBone {boneName=thigh_twist_01_r}}]
    constraintWeights [1,1,1]
   }
   n8 = Anim Dynamics - Bone: dyn_grenade(ComponentPose=@n7, Alpha=1, ExternalForce="0,0,0") {
    angularDampingOverride=0.985
    angularSpringConstant=100
    bAngularSpring=true
    bGravityOverrideInSimSpace=true
    bOverrideAngularDamping=true
    gravityScale=0.3
    numSolverIterationsPostUpdate=4
    numSolverIterationsPreUpdate=16
    simulationSpace=BoneRelative
    alphaBoolBlend {blendOption=Linear}
    boundBone {boneName=dyn_grenade}
    componentAppliedLinearAccClamp=<10000 10000 10000>
    componentLinearAccScale=<0.1 0.1 0.1>
    componentLinearVelScale=<0.1 0.1 0.1>
    gravityOverride {z=-1100}
    physicsBodyDefinitions [{
      sphereCollisionRadius=0
      boundBone {boneName=dyn_grenade}
      boxExtents=<15 15 20>
      constraintSetup {angularLimitsMax=<20 0 20>, angularLimitsMin=<-20 0 -20>, angularTarget=<0 0 0>}
      localJointOffset=<-3 0 -3>
     }]
    relativeSpaceBone {boneName=root}
   }
   n9 = "Transform (Modify) Bone - Bone: dyn_dress_back_01_l_MSTR"(ComponentPose=@n8, Alpha=1, Translation="0,-2,0", Rotation="0,0,0", Scale="1,1,1") {
    rotationMode=BMM_Additive
    rotationSpace=BCS_BoneSpace
    translationMode=BMM_Additive
    translationSpace=BCS_BoneSpace
    alphaBoolBlend {blendOption=Linear}
    boneToModify {boneName=dyn_dress_back_01_l_MSTR}
    translation {y=-2}
   }
   n10 = "Constraint - dyn_dress_front_01_l_MSTR (P)"(ComponentPose=@n9, Alpha=1, ConstraintWeights_0=1) {
    alphaBoolBlend {blendOption=Linear}
    boneToModify {boneName=dyn_dress_front_01_l_MSTR}
    constraintSetup [{targetBone {boneName=thigh_twist_01_l}}]
    constraintWeights [1]
   }
   n11 = "Transform (Modify) Bone - Bone: dyn_dress_side_01_l_MSTR"!(ComponentPose=@n10, bAlphaBoolEnabled:bool=True, Translation="0,0,0", Rotation="16,0,0.000", Scale="1,1,1") -> (Pose) {
    alphaInputType=Bool
    rotationMode=BMM_Additive
    rotationSpace=BCS_BoneSpace
    translationMode=BMM_Additive
    translationSpace=BCS_BoneSpace
    alphaBoolBlend {blendOption=Linear}
    boneToModify {boneName=dyn_dress_side_01_l_MSTR}
    rotation {pitch=16}
   }
   Save cached pose 'SavedPose3377'(Pose=Component To Local(ComponentPose=@n11)) {cacheName=SavedPose3377}
  ]}
 EventGraph {nodes [
   Event BlueprintUpdateAnimation()
   > n1 = Sequence()
   > n2 = Branch(Condition=Get bIsPlayingEmote())
   > Set Tie Side Alpha(Tie Side Alpha:double=0) -> (Output_Get:double)
   n3 = GetDeltaTransformFromRefPose(self=GetOwnerSkeletalMeshComponent(), BoneName=thigh_r, BaseName=pelvis)
   n4 = MapRangeClamped(Value=BreakRotator(InRot=BreakTransform(InTransform=@n3).Rotation).Yaw, InRangeA=0, InRangeB=-80)
   > Set Tie Front EM Alpha(Tie Front EM Alpha:double=@n4) -> (Output_Get:double)
   n5 = BreakTransform(InTransform=GetDeltaTransformFromRefPose(self=GetOwnerSkeletalMeshComponent(), BoneName=thigh_r, BaseName=pelvis))
   Set Tie Side Alpha(execute=@n2.else, Tie Side Alpha=MapRangeClamped(Value=BreakRotator(InRot=@n5.Rotation).Pitch, InRangeA=0, InRangeB=25))
   > Set Tie Front Alpha(Tie Front Alpha:double=@n4) -> (Output_Get:double)
   n6 = Branch(execute=@n1.then_1, Condition=Get bIsSkyDiving())
   > Set SD Alpha(SD Alpha:double=0.4) -> (Output_Get:double)
   Set SD Alpha(execute=@n6.else, SD Alpha=1)
  ]}
]
variables [
 SD Alpha {type=number}
 Tie Front Alpha {type=number}
 Tie Front EM Alpha {type=number}
 Tie Side Alpha {type=number}
]
