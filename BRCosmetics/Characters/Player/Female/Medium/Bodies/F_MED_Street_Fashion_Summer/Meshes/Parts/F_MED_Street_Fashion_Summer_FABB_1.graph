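# Physics asset for the dynamic hair of the F_MED_Street_Fashion_Summer female body cosmetic (FABB_1 part).
# Two back chains, each a master bone plus five capsule segments, and a two-bone side chain hang from the head through constraints with mostly narrow swing and twist limits.
# Head, neck, upper spine and right clavicle supply body shapes, with free angular constraints between neck, spine and clavicle; no left clavicle body exists.
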PhysicsAsset /BRCosmetics/Characters/Player/Female/Medium/Bodies/F_MED_Street_Fashion_Summer/Meshes/Parts/F_MED_Street_Fashion_Summer_FABB_1
bodies {
 clavicle_r [{shapeType=Capsule, length=19, radius=5.505, center=<-7.7468 -0.7223 -0.1266>, rotation=<rot 84.989 180 180>}]
 dyn_hairBack_01_l [{shapeType=Capsule, length=4, radius=2, center=<-1.1349 0 0>, rotation=<rot 90 0 0>}]
 dyn_hairBack_01_r [{shapeType=Capsule, length=4, radius=2, center=<1.3241 0 0.0001>, rotation=<rot 90 0 0>}]
 dyn_hairBack_02_l [{shapeType=Capsule, length=4, radius=2, center=<-2.5974 0 0>, rotation=<rot 90 0 0>}]
 dyn_hairBack_02_r [{shapeType=Capsule, length=4, radius=2, center=<2.448 1.1454 0>, rotation=<rot 90 0 0>}]
 dyn_hairBack_03_l [{shapeType=Capsule, length=4, radius=2, center=<-2.5974 0 2.4478>, rotation=<rot 90 0 0>}]
 dyn_hairBack_03_r [{shapeType=Capsule, length=4, radius=2, center=<2.448 0.0001 -2.2716>, rotation=<rot 90 0 0>}]
 dyn_hairBack_04_l [{shapeType=Capsule, length=4, radius=2, center=<-2.5974 0 1.1104>, rotation=<rot 90 0 0>}]
 dyn_hairBack_04_r [{shapeType=Capsule, length=4, radius=2, center=<2.448 0.0001 -1.6699>, rotation=<rot 69.9953 -180 -179.9999>}]
 dyn_hairBack_05_l [{shapeType=Capsule, length=4, radius=2, center=<-3.6223 -1.1388 0.0001>, rotation=<rot 90 0 0>}]
 dyn_hairBack_05_r [{shapeType=Capsule, length=4, radius=2, center=<3.471 1.2351 -0.0001>, rotation=<rot 90 0 0>}]
 dyn_hairBack_MSTR_l [{shapeType=Capsule, length=1, radius=1, rotation=<rot 90 0 0>}]
 dyn_hairBack_MSTR_r [{shapeType=Capsule, length=1, radius=1, rotation=<rot 90 0 0>}]
 dyn_hair_side_01 [{shapeType=Capsule, length=3.4975, radius=1.0058, center=<3.1514 0 0>, rotation=<rot 90 0 0>}]
 dyn_hair_side_02 [{shapeType=Capsule, length=2.63, radius=0.755, center=<2.471 0 -0.0001>, rotation=<rot 90 0 0>}]
 head [{shapeType=Sphere, radius=7}]
 neck_01 [{shapeType=Capsule, length=3.63, radius=5.255, center=<-0.6603 1.7172 0>, rotation=<rot 0 -90 0>}]
 neck_02 [{shapeType=Capsule, length=1.875, radius=5.13, center=<0 0.8322 0>, rotation=<rot 0 -90 0>}]
 spine_05 [{shapeType=Capsule, length=40.6583, radius=10.375, center=<3.0248 3.0111 0.0001>, rotation=<rot 0 -90 0>}]
}
constraints [
 {bone1Name=head, bone2Name=neck_02, swing1LimitDegrees=45, swing2LimitDegrees=45, twistLimitDegrees=45}
 {bone1Name=dyn_hair_side_01, bone2Name=head, swing1LimitDegrees=5, swing2LimitDegrees=5, twistLimitDegrees=2}
 {bone1Name=dyn_hair_side_02, bone2Name=dyn_hair_side_01, swing1LimitDegrees=20, swing2LimitDegrees=20, twistLimitDegrees=5}
 {bone1Name=dyn_hairBack_MSTR_r, bone2Name=head, swing1LimitDegrees=45, swing2LimitDegrees=45, twistLimitDegrees=5}
 {bone1Name=dyn_hairBack_01_r, bone2Name=dyn_hairBack_MSTR_r, swing1LimitDegrees=30, swing2LimitDegrees=10, twistLimitDegrees=2}
 {bone1Name=dyn_hairBack_02_r, bone2Name=dyn_hairBack_01_r, swing1LimitDegrees=15, swing2LimitDegrees=3, twistLimitDegrees=2}
 {bone1Name=dyn_hairBack_03_r, bone2Name=dyn_hairBack_02_r, swing1LimitDegrees=15, swing2LimitDegrees=5, twistLimitDegrees=2}
 {bone1Name=dyn_hairBack_04_r, bone2Name=dyn_hairBack_03_r, swing1LimitDegrees=15, swing2LimitDegrees=5, twistLimitDegrees=2}
 {bone1Name=dyn_hairBack_05_r, bone2Name=dyn_hairBack_04_r, swing1LimitDegrees=20, swing2LimitDegrees=10, twistLimitDegrees=2}
 {bone1Name=dyn_hairBack_MSTR_l, bone2Name=head, swing1LimitDegrees=45, swing2LimitDegrees=45, twistLimitDegrees=5}
 {bone1Name=dyn_hairBack_01_l, bone2Name=dyn_hairBack_MSTR_l, swing1LimitDegrees=30, swing2LimitDegrees=10, twistLimitDegrees=2}
 {bone1Name=dyn_hairBack_02_l, bone2Name=dyn_hairBack_01_l, swing1LimitDegrees=15, swing2LimitDegrees=5, twistLimitDegrees=2}
 {bone1Name=dyn_hairBack_03_l, bone2Name=dyn_hairBack_02_l, swing1LimitDegrees=15, swing2LimitDegrees=10, twistLimitDegrees=2}
 {bone1Name=dyn_hairBack_04_l, bone2Name=dyn_hairBack_03_l, swing1LimitDegrees=15, swing2LimitDegrees=10, twistLimitDegrees=2}
 {bone1Name=dyn_hairBack_05_l, bone2Name=dyn_hairBack_04_l, swing1LimitDegrees=20, swing2LimitDegrees=10, twistLimitDegrees=2}
 {
  bone1Name=neck_01
  bone2Name=spine_05
  swing1LimitDegrees=45
  swing1Motion=Free
  swing2LimitDegrees=45
  swing2Motion=Free
  twistLimitDegrees=45
  twistMotion=Free
 }
 {
  bone1Name=neck_02
  bone2Name=neck_01
  swing1LimitDegrees=45
  swing1Motion=Free
  swing2LimitDegrees=45
  swing2Motion=Free
  twistLimitDegrees=45
  twistMotion=Free
 }
 {
  bone1Name=clavicle_r
  bone2Name=spine_05
  swing1LimitDegrees=45
  swing1Motion=Free
  swing2LimitDegrees=45
  swing2Motion=Free
  twistLimitDegrees=45
  twistMotion=Free
 }
]
