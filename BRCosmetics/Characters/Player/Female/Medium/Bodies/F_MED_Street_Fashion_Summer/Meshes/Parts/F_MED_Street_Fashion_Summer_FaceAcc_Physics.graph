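# Physics asset for the FaceAcc mesh part of the medium female Street Fashion Summer player outfit, setting up hair bones for physics.
# Left and right back-hair chains and a two-bone side strand hang from the head as capsules.
# Narrow twist and uneven swing limits let back hair bend mostly along one axis, more loosely toward the tips.
# A head sphere and capsules on neck, clavicles and upper spine form the body, mostly joined by free angular constraints.
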PhysicsAsset /BRCosmetics/Characters/Player/Female/Medium/Bodies/F_MED_Street_Fashion_Summer/Meshes/Parts/F_MED_Street_Fashion_Summer_FaceAcc_Physics
bodies {
 clavicle_l [{shapeType=Capsule, length=18.25, radius=2.5, center=<7.499 -4.4459 2.9525>, rotation=<rot 84.9889 179.9999 179.9999>}]
 clavicle_r [{shapeType=Capsule, length=18.75, radius=2.5, center=<-7.5025 0.0105 -2.9129>, rotation=<rot 84.9889 179.9999 179.9999>}]
 dyn_hairBack_01_l [{shapeType=Capsule, length=4, radius=2, center=<-1.1349 0 0>, rotation=<rot 90 0 0>}]
 dyn_hairBack_01_r [{shapeType=Capsule, length=4, radius=2, center=<1.3241 0 0.0001>, rotation=<rot 90 0 0>}]
 dyn_hairBack_02_l [{shapeType=Capsule, length=4, radius=2, center=<-2.5974 0 0>, rotation=<rot 90 0 0>}]
 dyn_hairBack_02_r [{shapeType=Capsule, length=4, radius=2, center=<2.448 1.1454 0>, rotation=<rot 90 0 0>}]
 dyn_hairBack_03_l [
  {shapeType=Capsule, length=6, radius=2, center=<-4.9258 1.5361 2.4478>, rotation=<rot 90 0 9.9999>}
  {
   shapeType=Capsule
   length=9.25
   radius=3
   shapeName=dyn_hairBack_03_l_capsule
   center=<-1.316 -0.3211 3.5098>
   rotation=<rot 90 129.2614 149.2616>
  }
 ]
 dyn_hairBack_03_r [
  {shapeType=Capsule, length=4, radius=2, center=<2.448 0.0001 -2.2716>, rotation=<rot 90 0 0>}
  {
   shapeType=Capsule
   length=5.5
   radius=3
   shapeName=dyn_hairBack_03_r_capsule
   center=<3.5847 1.6496 -3.0218>
   rotation=<rot 90 180 -164.9985>
  }
 ]
 dyn_hairBack_04_l [{shapeType=Capsule, length=6, radius=3, center=<-2.5974 2 0>, rotation=<rot 90 0 0>}]
 dyn_hairBack_04_r [{shapeType=Capsule, length=4, radius=2, center=<2.448 0.8977 -1.67>, rotation=<rot 69.9953 -179.9999 -179.9999>}]
 dyn_hairBack_05_l [{shapeType=Capsule, length=6, radius=3, center=<-3.6223 1 0>, rotation=<rot 90 0 0>}]
 dyn_hairBack_05_r [{shapeType=Capsule, length=4, radius=2, center=<3.471 1.2351 -0.0001>, rotation=<rot 90 0 0>}]
 dyn_hairBack_MSTR_l [{shapeType=Capsule, length=1, radius=1, rotation=<rot 90 0 0>}]
 dyn_hairBack_MSTR_r [{shapeType=Capsule, length=1, radius=1, rotation=<rot 90 0 0>}]
 dyn_hair_side_01 [{shapeType=Capsule, length=3.4975, radius=1.0058, center=<3.1514 0 0>, rotation=<rot 90 0 0>}]
 dyn_hair_side_02 [{shapeType=Capsule, length=2.63, radius=0.755, center=<2.471 0 -0.0001>, rotation=<rot 90 0 0>}]
 head [{shapeType=Sphere, radius=7}]
 neck_01 [{shapeType=Capsule, length=3.63, radius=5.255, center=<-0.6603 1.7554 0>, rotation=<rot 0 -90 0>}]
 neck_02 [{shapeType=Capsule, length=1.875, radius=5.13, center=<0 0.6909 0>, rotation=<rot 0 -90 0>}]
 spine_04 [{shapeType=Capsule, length=34.505, radius=9.755, shapeName=spine_04_capsule, rotation=<rot 0 -90 0>}]
 spine_05 [{shapeType=Capsule, length=36.9083, radius=9.125, center=<3.1319 2.6867 0.0001>, rotation=<rot 0 -90 0>}]
}
constraints [
 {bone1Name=head, bone2Name=neck_02, swing1LimitDegrees=45, swing2LimitDegrees=45, twistLimitDegrees=45}
 {bone1Name=dyn_hair_side_01, bone2Name=head, swing1LimitDegrees=5, swing2LimitDegrees=5, twistLimitDegrees=2}
 {bone1Name=dyn_hair_side_02, bone2Name=dyn_hair_side_01, swing1LimitDegrees=20, swing2LimitDegrees=20, twistLimitDegrees=5}
 {bone1Name=dyn_hairBack_MSTR_r, bone2Name=head, swing1LimitDegrees=45, swing2LimitDegrees=45, twistLimitDegrees=5}
 {bone1Name=dyn_hairBack_01_r, bone2Name=dyn_hairBack_MSTR_r, swing1LimitDegrees=20, swing2LimitDegrees=10, twistLimitDegrees=2}
 {bone1Name=dyn_hairBack_02_r, bone2Name=dyn_hairBack_01_r, swing1LimitDegrees=25, swing2LimitDegrees=3, twistLimitDegrees=2}
 {bone1Name=dyn_hairBack_03_r, bone2Name=dyn_hairBack_02_r, swing1LimitDegrees=25, swing2LimitDegrees=5, twistLimitDegrees=2}
 {bone1Name=dyn_hairBack_04_r, bone2Name=dyn_hairBack_03_r, swing1LimitDegrees=55, swing2LimitDegrees=5, twistLimitDegrees=2}
 {bone1Name=dyn_hairBack_05_r, bone2Name=dyn_hairBack_04_r, swing1LimitDegrees=55, swing2LimitDegrees=10, twistLimitDegrees=2}
 {bone1Name=dyn_hairBack_MSTR_l, bone2Name=head, swing1LimitDegrees=45, swing2LimitDegrees=45, twistLimitDegrees=5}
 {bone1Name=dyn_hairBack_01_l, bone2Name=dyn_hairBack_MSTR_l, swing1LimitDegrees=20, swing2LimitDegrees=10, twistLimitDegrees=2}
 {bone1Name=dyn_hairBack_02_l, bone2Name=dyn_hairBack_01_l, swing1LimitDegrees=25, swing2LimitDegrees=5, twistLimitDegrees=2}
 {bone1Name=dyn_hairBack_03_l, bone2Name=dyn_hairBack_02_l, swing1LimitDegrees=25, swing2LimitDegrees=10, twistLimitDegrees=2}
 {bone1Name=dyn_hairBack_04_l, bone2Name=dyn_hairBack_03_l, swing1LimitDegrees=55, swing2LimitDegrees=10, twistLimitDegrees=2}
 {bone1Name=dyn_hairBack_05_l, bone2Name=dyn_hairBack_04_l, swing1LimitDegrees=55, swing2LimitDegrees=10, twistLimitDegrees=2}
 {
  bone1Name=neck_01
  bone2Name=spine_05
  swing1LimitDegrees=45
  swing1Motion=Free
  swing2LimitDegrees=45
  swing2Motion=Free
  twistLimitDegrees=45
  twistMotion=Free
 }
 {
  bone1Name=neck_02
  bone2Name=neck_01
  swing1LimitDegrees=45
  swing1Motion=Free
  swing2LimitDegrees=45
  swing2Motion=Free
  twistLimitDegrees=45
  twistMotion=Free
 }
 {
  bone1Name=clavicle_r
  bone2Name=spine_05
  swing1LimitDegrees=45
  swing1Motion=Free
  swing2LimitDegrees=45
  swing2Motion=Free
  twistLimitDegrees=45
  twistMotion=Free
 }
 {
  bone1Name=clavicle_l
  bone2Name=spine_05
  swing1LimitDegrees=45
  swing1Motion=Free
  swing2LimitDegrees=45
  swing2Motion=Free
  twistLimitDegrees=45
  twistMotion=Free
 }
 {
  bone1Name=spine_05
  bone2Name=spine_04
  swing1LimitDegrees=45
  swing1Motion=Free
  swing2LimitDegrees=45
  swing2Motion=Free
  twistLimitDegrees=45
  twistMotion=Free
 }
]
