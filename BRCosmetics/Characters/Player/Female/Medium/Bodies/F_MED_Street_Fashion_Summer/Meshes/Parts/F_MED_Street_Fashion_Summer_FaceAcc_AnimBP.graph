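# Anim Blueprint for the Summer Street Fashion female FaceAcc part, adding secondary motion to back hair and earrings.
# The pose copied from the attached parent gets AnimDynamics earrings and hair-root offsets, weighted by head-to-spine rotation and weapon type.
# RigidBody hair simulation varies for backpack (non-cape backbling) and skydiving.
# With rigid bodies disabled, hair falls back to Look At plus per-bone AnimDynamics chains.
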AnimBlueprint /BRCosmetics/Characters/Player/Female/Medium/Bodies/F_MED_Street_Fashion_Summer/Meshes/Parts/F_MED_Street_Fashion_Summer_FaceAcc_AnimBP
target_skeleton=&/F_MED_Street_Fashion_Summer_FaceAcc_Skeleton
graphs [
 AnimGraph {nodes [
   Local To Component(LocalPose=Use cached pose 'Backpack'())
   "Transform (Modify) Bone - Bone: dyn_hairBack_03_l"!(ComponentPose, Alpha=1, Rotation="-1.279,-35.854,4.707") -> (Pose) {
    rotationMode=BMM_Additive
    rotationSpace=BCS_BoneSpace
    translationMode=BMM_Additive
    translationSpace=BCS_BoneSpace
    alphaBoolBlend {blendOption=Linear}
    boneToModify {boneName=dyn_hairBack_03_l}
    rotation=<rot -1.279 -35.854 4.707>
   }
   n1 = RigidBody(ComponentPose=Local To Component(LocalPose=Use cached pose 'No Backpack'())) {
    bOverrideWorldGravity=true
    overridePhysicsAsset=&/F_MED_Street_Fashion_Summer_FaceAcc_Physics
    simulationSpace=BaseBoneSpace
    alphaBoolBlend {blendOption=Linear}
    baseBoneRef {boneName=spine_05}
    componentLinearAccScale=<0.4 0.4 0.4>
    componentLinearVelScale=<0.2 0.2 0.2>
    overrideWorldGravity {z=-980}
   }
   n2 = "Transform (Modify) Bone - Bone: dyn_hairBack_04_l"!(ComponentPose=@n1, Alpha=1, Rotation="0,-20,0") -> (Pose) {
    rotationMode=BMM_Replace
    rotationSpace=BCS_BoneSpace
    translationMode=BMM_Additive
    translationSpace=BCS_BoneSpace
    alphaBoolBlend {blendOption=Linear}
    boneToModify {boneName=dyn_hairBack_04_l}
    rotation {yaw=-20}
   }
   n3 = "Transform (Modify) Bone - Bone: dyn_hairBack_04_r"!(ComponentPose=@n2, Alpha=1, Rotation="-0.000,-30,-0.000") -> (Pose) {
    rotationMode=BMM_Replace
    rotationSpace=BCS_BoneSpace
    translationMode=BMM_Additive
    translationSpace=BCS_BoneSpace
    alphaBoolBlend {blendOption=Linear}
    boneToModify {boneName=dyn_hairBack_04_r}
    rotation {yaw=-30}
   }
   n4 = RigidBody(ComponentPose=Local To Component(LocalPose=Use cached pose 'No Backpack'())) {
    bClampLinearTranslationLimitToRefPose=true
    bOverrideWorldGravity=true
    overridePhysicsAsset=&F_MED_Street_Fashion_Summer_FaceAcc_Physics
    simulationSpace=BaseBoneSpace
    alphaBoolBlend {blendOption=Linear}
    baseBoneRef {boneName=spine_05}
    componentAppliedLinearAccClamp=<5000 5000 5000>
    componentLinearAccScale=<0.3 0.3 0.05>
    componentLinearVelScale=<0.15 0.15 0.075>
    overrideWorldGravity {z=-500}
   }
   n5 = Blend Poses by bool(BlendPose_0=Component To Local(ComponentPose=@n4), BlendPose_1=Component To Local(ComponentPose=@n3), bActiveValue=Get bIsSkyDiving()) {blendType=Linear}
   n6 = "Transform (Modify) Bone - Bone: dyn_hairBack_MSTR_r"(
    ComponentPose=Local To Component(LocalPose=Use cached pose 'Pose'())
    Alpha=Get UP Hair Rot Alpha() -> (UP Hair Rot Alpha:double)
    Translation="-3,0,1"
    Rotation="0,-20,0.000"
    Scale="1,1,1"
   ) {
    rotationMode=BMM_Additive
    rotationSpace=BCS_BoneSpace
    translationMode=BMM_Additive
    translationSpace=BCS_BoneSpace
    alphaBoolBlend {blendOption=Linear}
    boneToModify {boneName=dyn_hairBack_MSTR_r}
    rotation {yaw=-20}
    translation {x=-3, z=1}
   }
   n7 = RigidBody(ComponentPose=@n6) {
    bClampLinearTranslationLimitToRefPose=true
    bOverrideWorldGravity=true
    overridePhysicsAsset=&/F_MED_Street_Fashion_Summer_FABB_1
    simulationSpace=BaseBoneSpace
    alphaBoolBlend {blendOption=Linear}
    baseBoneRef {boneName=spine_05}
    componentAppliedLinearAccClamp=<3000 3000 300>
    componentLinearAccScale=<0.4 0.4 0.05>
    componentLinearVelScale=<0.15 0.15 0.075>
    overrideWorldGravity {z=-420}
   }
   n8 = Component To Local(ComponentPose=@n7)
   n9 = Get Wearing Backpack() -> (Wearing Backpack:bool)
   n10 = "Transform (Modify) Bone - Bone: dyn_hairBack_MSTR_l"!(ComponentPose=Local To Component(LocalPose=Use cached pose 'No Backpack'()), Alpha=1, Rotation="0,-10,0") -> (Pose) {
    rotationMode=BMM_Replace
    rotationSpace=BCS_BoneSpace
    translationSpace=BCS_BoneSpace
    alphaBoolBlend {blendOption=Linear}
    boneToModify {boneName=dyn_hairBack_MSTR_l}
    rotation {yaw=-10}
    translation {x=-3}
   }
   n11 = "Transform (Modify) Bone - Bone: dyn_hairBack_MSTR_r"!(ComponentPose=@n10, Alpha=1, Rotation="0,-15,0") -> (Pose) {
    rotationMode=BMM_Replace
    rotationSpace=BCS_BoneSpace
    translationSpace=BCS_BoneSpace
    alphaBoolBlend {blendOption=Linear}
    boneToModify {boneName=dyn_hairBack_MSTR_r}
    rotation {yaw=-15}
    translation {x=-3}
   }
   n12 = Look At - Bone: dyn_hairBack_03_l(ComponentPose=@n11, Alpha=1) {
    alphaBoolBlend {blendOption=Linear}
    boneToModify {boneName=dyn_hairBack_03_l}
    lookAtLocation=<33.0518 -0.4024 -1.1159>
    lookAtTarget {boneReference {boneName=spine_03}}
    lookAt_Axis {axis=<1 0>}
   }
   n13 = Look At - Bone: dyn_hairBack_03_r(ComponentPose=@n12, Alpha=1) {
    alphaBoolBlend {blendOption=Linear}
    boneToModify {boneName=dyn_hairBack_03_r}
    lookAtLocation=<0.9852 -50.6342 8.2981>
    lookAtTarget {boneReference {boneName=spine_03}}
    lookAt_Axis {axis=<1 0>}
   }
   n14 = Anim Dynamics - Bone: dyn_hairBack_01_l(ComponentPose=@n13, Alpha=1, ExternalForce="0,0,0") {
    angularDampingOverride=0.9999
    angularSpringConstant=500
    bAngularSpring=true
    bGravityOverrideInSimSpace=true
    bOverrideAngularDamping=true
    simulationSpace=BoneRelative
    alphaBoolBlend {blendOption=Linear}
    boundBone {boneName=dyn_hairBack_01_l}
    componentAppliedLinearAccClamp=<3500 3500 3500>
    componentLinearAccScale=<1.2 1.2 1.2>
    componentLinearVelScale=<1.2 1.2 1.2>
    gravityOverride {z=-980}
    physicsBodyDefinitions [{
      sphereCollisionRadius=0
      boundBone {boneName=dyn_hairBack_01_l}
      boxExtents=<15 20 20>
      constraintSetup {angularLimitsMax=<5 20 3>, angularLimitsMin=<-5 -20 -15>, angularTarget=<0 0 0>}
      localJointOffset=<-5 0 0>
     }]
    relativeSpaceBone {boneName=root}
   }
   n15 = Anim Dynamics - Bone: dyn_hairBack_02_l(ComponentPose=@n14, Alpha=1, ExternalForce="0,0,0") {
    angularDampingOverride=0.9999
    angularSpringConstant=500
    bAngularSpring=true
    bGravityOverrideInSimSpace=true
    bOverrideAngularDamping=true
    simulationSpace=BoneRelative
    alphaBoolBlend {blendOption=Linear}
    boundBone {boneName=dyn_hairBack_02_l}
    componentAppliedLinearAccClamp=<3500 3500 3500>
    componentLinearAccScale=<1.2 1.2 1.2>
    componentLinearVelScale=<1.2 1.2 1.2>
    gravityOverride {z=-980}
    physicsBodyDefinitions [{
      sphereCollisionRadius=0
      boundBone {boneName=dyn_hairBack_02_l}
      boxExtents=<15 20 20>
      constraintSetup {angularLimitsMax=<5 20 10>, angularLimitsMin=<-5 -20 -10>, angularTarget=<0 0 0>}
      localJointOffset=<-5 0 0>
     }]
    relativeSpaceBone {boneName=root}
   }
   n16 = Anim Dynamics - Bone: dyn_hairBack_03_l(ComponentPose=@n15, Alpha=1, ExternalForce="0,0,0") {
    angularDampingOverride=0.9999
    angularSpringConstant=100
    bAngularSpring=true
    bGravityOverrideInSimSpace=true
    bOverrideAngularDamping=true
    simulationSpace=BoneRelative
    alphaBoolBlend {blendOption=Linear}
    boundBone {boneName=dyn_hairBack_03_l}
    componentAppliedLinearAccClamp=<3500 3500 3500>
    componentLinearAccScale=<1.2 1.2 1.2>
    componentLinearVelScale=<1.2 1.2 1.2>
    gravityOverride {z=-980}
    physicsBodyDefinitions [{
      sphereCollisionRadius=0
      boundBone {boneName=dyn_hairBack_03_l}
      boxExtents=<15 20 20>
      constraintSetup {angularLimitsMax=<5 15 2>, angularLimitsMin=<-5 -15 -25>, angularTarget=<0 0 0>}
      localJointOffset=<-5 0 0>
     }]
    relativeSpaceBone {boneName=root}
   }
   n17 = Anim Dynamics - Bone: dyn_hairBack_04_l(ComponentPose=@n16, Alpha=1, ExternalForce="0,0,0") {
    angularDampingOverride=0.9999
    angularSpringConstant=100
    bAngularSpring=true
    bGravityOverrideInSimSpace=true
    bOverrideAngularDamping=true
    simulationSpace=BoneRelative
    alphaBoolBlend {blendOption=Linear}
    boundBone {boneName=dyn_hairBack_04_l}
    componentAppliedLinearAccClamp=<3500 3500 3500>
    componentLinearAccScale=<1.2 1.2 1.2>
    componentLinearVelScale=<1.2 1.2 1.2>
    gravityOverride {z=-980}
    physicsBodyDefinitions [{
      sphereCollisionRadius=0
      boundBone {boneName=dyn_hairBack_04_l}
      boxExtents=<15 20 20>
      constraintSetup {angularLimitsMax=<5 15 2>, angularLimitsMin=<-5 -15 -25>, angularTarget=<0 0 0>}
      localJointOffset=<-5 0 0>
     }]
    relativeSpaceBone {boneName=root}
   }
   n18 = Anim Dynamics - Bone: dyn_hairBack_05_l(ComponentPose=@n17, Alpha=1, ExternalForce="0,0,0") {
    angularDampingOverride=0.9999
    angularSpringConstant=100
    bAngularSpring=true
    bGravityOverrideInSimSpace=true
    bOverrideAngularDamping=true
    simulationSpace=BoneRelative
    alphaBoolBlend {blendOption=Linear}
    boundBone {boneName=dyn_hairBack_05_l}
    componentAppliedLinearAccClamp=<3500 3500 3500>
    componentLinearAccScale=<1.2 1.2 1.2>
    componentLinearVelScale=<1.2 1.2 1.2>
    gravityOverride {z=-980}
    physicsBodyDefinitions [{
      sphereCollisionRadius=0
      boundBone {boneName=dyn_hairBack_05_l}
      boxExtents=<15 20 20>
      constraintSetup {angularLimitsMax=<5 15 2>, angularLimitsMin=<-5 -15 -25>, angularTarget=<0 0 0>}
      localJointOffset=<-5 0 0>
     }]
    relativeSpaceBone {boneName=root}
   }
   n19 = Anim Dynamics - Bone: dyn_hairBack_01_r(ComponentPose=@n18, Alpha=1, ExternalForce="0,0,0") {
    angularDampingOverride=0.9999
    angularSpringConstant=500
    bAngularSpring=true
    bGravityOverrideInSimSpace=true
    bOverrideAngularDamping=true
    simulationSpace=BoneRelative
    alphaBoolBlend {blendOption=Linear}
    boundBone {boneName=dyn_hairBack_01_r}
    componentAppliedLinearAccClamp=<3500 3500 3500>
    componentLinearAccScale=<1.2 1.2 1.2>
    componentLinearVelScale=<1.2 1.2 1.2>
    gravityOverride {z=-980}
    physicsBodyDefinitions [{
      sphereCollisionRadius=0
      boundBone {boneName=dyn_hairBack_01_r}
      boxExtents=<15 20 20>
      constraintSetup {angularLimitsMax=<5 20 5>, angularLimitsMin=<-5 -20 -15>, angularTarget=<0 0 0>}
      localJointOffset=<2 0 0>
     }]
    relativeSpaceBone {boneName=root}
   }
   n20 = Anim Dynamics - Bone: dyn_hairBack_02_r(ComponentPose=@n19, Alpha=1, ExternalForce="0,0,0") {
    angularDampingOverride=0.9999
    angularSpringConstant=500
    bAngularSpring=true
    bGravityOverrideInSimSpace=true
    bOverrideAngularDamping=true
    simulationSpace=BoneRelative
    alphaBoolBlend {blendOption=Linear}
    boundBone {boneName=dyn_hairBack_02_r}
    componentAppliedLinearAccClamp=<3500 3500 3500>
    componentLinearAccScale=<1.2 1.2 1.2>
    componentLinearVelScale=<1.2 1.2 1.2>
    gravityOverride {z=-980}
    physicsBodyDefinitions [{
      sphereCollisionRadius=0
      boundBone {boneName=dyn_hairBack_02_r}
      boxExtents=<15 20 20>
      constraintSetup {angularLimitsMax=<5 20 5>, angularLimitsMin=<-5 -20 -15>, angularTarget=<0 0 0>}
      localJointOffset=<5 0 0>
     }]
    relativeSpaceBone {boneName=root}
   }
   n21 = Anim Dynamics - Bone: dyn_hairBack_03_r(ComponentPose=@n20, Alpha=1, ExternalForce="0,0,0") {
    angularDampingOverride=0.9999
    angularSpringConstant=100
    bAngularSpring=true
    bGravityOverrideInSimSpace=true
    bOverrideAngularDamping=true
    simulationSpace=BoneRelative
    alphaBoolBlend {blendOption=Linear}
    boundBone {boneName=dyn_hairBack_03_r}
    componentAppliedLinearAccClamp=<3500 3500 3500>
    componentLinearAccScale=<1.2 1.2 1.2>
    componentLinearVelScale=<1.2 1.2 1.2>
    gravityOverride {z=-980}
    physicsBodyDefinitions [{
      sphereCollisionRadius=0
      boundBone {boneName=dyn_hairBack_03_r}
      boxExtents=<15 20 20>
      constraintSetup {angularLimitsMax=<5 15 2>, angularLimitsMin=<-5 -15 -25>, angularTarget=<0 0 0>}
      localJointOffset=<5 0 0>
     }]
    relativeSpaceBone {boneName=root}
   }
   n22 = Anim Dynamics - Bone: dyn_hairBack_04_r(ComponentPose=@n21, Alpha=1, ExternalForce="0,0,0") {
    angularDampingOverride=0.9999
    angularSpringConstant=100
    bAngularSpring=true
    bGravityOverrideInSimSpace=true
    bOverrideAngularDamping=true
    simulationSpace=BoneRelative
    alphaBoolBlend {blendOption=Linear}
    boundBone {boneName=dyn_hairBack_04_r}
    componentAppliedLinearAccClamp=<3500 3500 3500>
    componentLinearAccScale=<1.2 1.2 1.2>
    componentLinearVelScale=<1.2 1.2 1.2>
    gravityOverride {z=-980}
    physicsBodyDefinitions [{
      sphereCollisionRadius=0
      boundBone {boneName=dyn_hairBack_04_r}
      boxExtents=<15 20 20>
      constraintSetup {angularLimitsMax=<5 15 2>, angularLimitsMin=<-5 -15 -25>, angularTarget=<0 0 0>}
      localJointOffset=<5 0 0>
     }]
    relativeSpaceBone {boneName=root}
   }
   n23 = Anim Dynamics - Bone: dyn_hairBack_05_r(ComponentPose=@n22, Alpha=1, ExternalForce="0,0,0") {
    angularDampingOverride=0.9999
    angularSpringConstant=100
    bAngularSpring=true
    bGravityOverrideInSimSpace=true
    bOverrideAngularDamping=true
    simulationSpace=BoneRelative
    alphaBoolBlend {blendOption=Linear}
    boundBone {boneName=dyn_hairBack_05_r}
    componentAppliedLinearAccClamp=<3500 3500 3500>
    componentLinearAccScale=<1.2 1.2 1.2>
    componentLinearVelScale=<1.2 1.2 1.2>
    gravityOverride {z=-980}
    physicsBodyDefinitions [{
      sphereCollisionRadius=0
      boundBone {boneName=dyn_hairBack_05_r}
      boxExtents=<15 20 20>
      constraintSetup {angularLimitsMax=<5 15 2>, angularLimitsMin=<-5 -15 -25>, angularTarget=<0 0 0>}
      localJointOffset=<5 0 0>
     }]
    relativeSpaceBone {boneName=root}
   }
   n24 = Look At - Bone: dyn_hairBack_03_l(ComponentPose=@n23, Alpha=1) {
    alphaBoolBlend {blendOption=Linear}
    boneToModify {boneName=dyn_hairBack_03_l}
    lookAtLocation=<28.8191 78.8816 39.1749>
    lookAtTarget {boneReference {boneName=head}}
    lookAt_Axis {axis=<1 0>}
   }
   n25 = "Transform (Modify) Bone - Bone: dyn_hairBack_03_r"!(ComponentPose=@n24, Alpha=1, Rotation="5.310,-2.018,-0.746") -> (Pose) {
    rotationMode=BMM_Additive
    rotationSpace=BCS_BoneSpace
    translationMode=BMM_Additive
    translationSpace=BCS_BoneSpace
    alphaBoolBlend {blendOption=Linear}
    boneToModify {boneName=dyn_hairBack_03_r}
    rotation=<rot 5.31 -2.018 -0.746>
   }
   Output Pose(Result=Blend Poses by bool(
     BlendPose_0=Blend Poses by bool(BlendPose_0=@n8, BlendPose_1=@n5, bActiveValue=@n9) {blendType=Linear}
     BlendPose_1=Component To Local(ComponentPose=@n25)
     bActiveValue=Get bRigidBodyEnabled()
    ) {blendType=Linear}) {layerGroup=None}
   Save cached pose 'Backpack'(Pose=Component To Local()) {cacheName=Backpack}
   n26 = "Transform (Modify) Bone - Bone: dyn_hairBack_MSTR_r"!(
    ComponentPose=Local To Component(LocalPose=Use cached pose 'Pose'())
    Alpha=Get UP Hair Rot Alpha()
    Translation:Vector="-3,0.000,0.000"
    Rotation="10,-15,0.000"
    Scale:Vector="1,1,1"
   ) -> (Pose) {
    rotationMode=BMM_Additive
    rotationSpace=BCS_BoneSpace
    translationMode=BMM_Additive
    translationSpace=BCS_BoneSpace
    alphaBoolBlend {blendOption=Linear}
    boneToModify {boneName=dyn_hairBack_MSTR_r}
    rotation {pitch=10, yaw=-15}
    translation {x=-3}
   }
   Save cached pose 'No Backpack'(Pose=Component To Local(ComponentPose=@n26)) {cacheName=No Backpack}
   n27 = Anim Dynamics - Bone: dyn_earring_r(ComponentPose=Local To Component(LocalPose=Copy Pose From Mesh() {bUseAttachedParent=true}), Alpha=1, ExternalForce="0,0,0") {
    angularDampingOverride=0.994
    angularSpringConstant=5
    bAngularSpring=true
    bGravityOverrideInSimSpace=true
    bOverrideAngularDamping=true
    bUseGravityOverride=true
    gravityScale=0.35
    simulationSpace=BoneRelative
    alphaBoolBlend {blendOption=Linear}
    boundBone {boneName=dyn_earring_r}
    componentAppliedLinearAccClamp=<10000 10000 10000>
    componentLinearAccScale=<0.4 0.4 0.4>
    componentLinearVelScale=<0.4 0.4 0.4>
    gravityOverride {z=-980}
    physicsBodyDefinitions [{
      sphereCollisionRadius=0
      boundBone {boneName=dyn_earring_r}
      boxExtents=<8 3 8>
      constraintSetup {angularLimitsMax=<2 30 30>, angularLimitsMin=<-2 -5 -30>, angularTarget=<1 0 0>}
      localJointOffset=<1 0 0>
     }]
    relativeSpaceBone {boneName=root}
   }
   n28 = Anim Dynamics - Bone: dyn_earring_l(ComponentPose=@n27, Alpha=1, ExternalForce="0,0,0") {
    angularDampingOverride=0.99
    angularSpringConstant=5
    bAngularSpring=true
    bGravityOverrideInSimSpace=true
    bOverrideAngularDamping=true
    bUseGravityOverride=true
    gravityScale=0.35
    simulationSpace=BoneRelative
    alphaBoolBlend {blendOption=Linear}
    boundBone {boneName=dyn_earring_l}
    componentAppliedLinearAccClamp=<10000 10000 10000>
    componentLinearAccScale=<0.1 0.1 0.1>
    componentLinearVelScale=<0.1 0.1 0.1>
    gravityOverride {z=-980}
    physicsBodyDefinitions [{
      sphereCollisionRadius=0
      boundBone {boneName=dyn_earring_l}
      boxExtents=<3 1 3>
      constraintSetup {angularLimitsMax=<0 3 3>, angularLimitsMin=<0 -3 -3>, angularTarget=<1 0 0>}
      localJointOffset=<1 0 0>
     }]
    relativeSpaceBone {boneName=root}
   }
   n29 = Get ADS Hair Alpha() -> (ADS Hair Alpha:double)
   n30 = "Transform (Modify) Bone - Bone: dyn_hairBack_MSTR_r"!(ComponentPose=@n28, Alpha=@n29, Translation="-3,3,2", Rotation="5,-5,0") -> (Pose) {
    rotationMode=BMM_Additive
    rotationSpace=BCS_BoneSpace
    translationMode=BMM_Additive
    translationSpace=BCS_BoneSpace
    alphaBoolBlend {blendOption=Linear}
    boneToModify {boneName=dyn_hairBack_MSTR_r}
    rotation {pitch=5, yaw=-5}
    translation=<-3 3 2>
   }
   n31 = "Transform (Modify) Bone - Bone: dyn_hairBack_MSTR_l"(ComponentPose=@n30, Alpha=@n29, Rotation="5,0.000,0.000") {
    rotationMode=BMM_Additive
    rotationSpace=BCS_BoneSpace
    translationMode=BMM_Additive
    translationSpace=BCS_BoneSpace
    alphaBoolBlend {blendOption=Linear}
    boneToModify {boneName=dyn_hairBack_MSTR_l}
    rotation {pitch=5}
   }
   n32 = Get DOWN Hair Rot Alpha() -> (DOWN Hair Rot Alpha:double)
   n33 = "Transform (Modify) Bone - Bone: dyn_hairBack_MSTR_r"!(ComponentPose=@n31, Alpha=@n32, Rotation="5,-10,0.000") -> (Pose) {
    rotationMode=BMM_Additive
    rotationSpace=BCS_BoneSpace
    translationMode=BMM_Additive
    translationSpace=BCS_BoneSpace
    alphaBoolBlend {blendOption=Linear}
    boneToModify {boneName=dyn_hairBack_MSTR_r}
    rotation {pitch=5, yaw=-10}
   }
   n34 = "Transform (Modify) Bone - Bone: dyn_hairBack_MSTR_l"!(ComponentPose=@n33, Alpha=@n32, Translation:Vector="-2,0,0", Rotation="0,15,0") -> (Pose) {
    rotationMode=BMM_Additive
    rotationSpace=BCS_BoneSpace
    translationSpace=BCS_BoneSpace
    alphaBoolBlend {blendOption=Linear}
    boneToModify {boneName=dyn_hairBack_MSTR_l}
    rotation {yaw=15}
    translation {x=-2}
   }
   n35 = "Transform (Modify) Bone - Bone: dyn_hairBack_MSTR_l"(ComponentPose=@n34, Alpha=Get UP Hair Rot Alpha(), Translation="3,0,0", Rotation="0,-15,0") {
    rotationMode=BMM_Additive
    rotationSpace=BCS_BoneSpace
    translationMode=BMM_Additive
    translationSpace=BCS_BoneSpace
    alphaBoolBlend {blendOption=Linear}
    boneToModify {boneName=dyn_hairBack_MSTR_l}
    rotation {yaw=-15}
    translation {x=3}
   }
   Save cached pose 'Pose'(Pose=Component To Local(ComponentPose=@n35)) {cacheName=Pose}
  ]}
 EventGraph {nodes [
   Event BlueprintUpdateAnimation()
   n1 = GetOwnerSkeletalMeshComponent()
   n2 = BreakRotator(InRot=BreakTransform(InTransform=GetDeltaTransformFromRefPose(self=@n1, BoneName=head, BaseName=spine_05)).Rotation)
   > Set UP Hair Rot Alpha(UP Hair Rot Alpha:double=MapRangeClamped(Value=@n2.Yaw, InRangeA=0, InRangeB=-40)) -> (Output_Get:double)
   > Set DOWN Hair Rot Alpha(DOWN Hair Rot Alpha:double=MapRangeClamped(Value=@n2.Yaw, InRangeA=0, InRangeB=14)) -> (Output_Get:double)
   > n3 = Cast To FortPlayerAnimInstance(Object=GetAnimInstance(self=GetOwnerSkeletalMeshComponent()))
   > Set ADS Hair Alpha(ADS Hair Alpha:double=0) -> (Output_Get:double)
   > n4 = Switch on EFortWeaponCoreAnimation!(execute, Selection:EFortWeaponCoreAnimation Enum=Get WeaponCoreAnim(self=@n3)) -> (
    Melee:Exec
    Pistol:Exec
    Shotgun:Exec
    PaperBlueprint:Exec
    Rifle:Exec
    MeleeOneHand:Exec
    MachinePistol:Exec
    RocketLauncher:Exec
    GrenadeLauncher:Exec
    GoingCommando:Exec
    AssaultRifle:Exec
    TacticalShotgun:Exec
    SniperRifle:Exec
    TrapPlacement:Exec
    ShoulderLauncher:Exec
    AbilityDecoTool:Exec
    Crossbow:Exec
    C4:Exec
    RemoteControl:Exec
    DualWield:Exec
    AR_BullPup:Exec
    AR_ForwardGrip:Exec
    MedPackPaddles:Exec
    SMG_P90:Exec
    AR_DrumGun:Exec
    Consumable_Small:Exec
    Consumable_Large:Exec
    Balloon:Exec
    MountedTurret:Exec
    CreativeTool:Exec
    ExplosiveBow:Exec
    AshtonIndigo:Exec
    AshtonChicago:Exec
    MeleeDualWield:Exec
    Unarmed:Exec
   )
   n5 = GetDeltaTransformFromRefPose(self=GetOwnerSkeletalMeshComponent(), BoneName=head, BaseName=spine_02)
   n6 = MapRangeClamped(Value=BreakRotator(InRot=BreakTransform(InTransform=@n5).Rotation).Pitch, InRangeA=0, InRangeB=18)
   Set ADS Hair Alpha(execute=[@n7.then, @n4.AssaultRifle, @n4.GoingCommando, @n4.Rifle, @n4.Shotgun], ADS Hair Alpha=@n6)
   n7 = Branch(execute=@n4.Pistol, Condition=Get Wearing Backpack())
   Set ADS Hair Alpha(execute=@n7.else, ADS Hair Alpha=float / float(A=@n6, B=3) {is=title K2Node_CallFunction})
   Event BlueprintInitializeAnimation()
   n8 = CheckForBackbling()
   > n9 = Set Wearing Backpack(Wearing Backpack:bool=AND Boolean(A=@n8, B=NOT Boolean(A=@n8.bIsBackblingACape))) -> (Output_Get:bool)
   > Set No Backpack(No Backpack:bool=NOT Boolean(A=@n9)) -> (Output_Get:bool)
   n10 = GetDeltaTransformFromRefPose(self=GetOwnerSkeletalMeshComponent(), BoneName=head, BaseName=spine_05)
   MapRangeClamped(Value=BreakRotator(InRot=BreakTransform(InTransform=@n10).Rotation).Pitch, InRangeA=8, InRangeB=40)
  ]}
]
variables [
 ADS Hair Alpha {type=number}
 DOWN Hair Rot Alpha {type=number}
 No Backpack {type=boolean}
 Pickaxe Alpha {type=number}
 RifleADSAlpha {type=number}
 UP Hair Rot Alpha {type=number}
 UP Hair Rot Alpha bp {type=number}
 Wearing Backpack {type=boolean}
]
